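# Widget Blueprint for one action-setting row in the HUD Layout Tool's Link Across feature, built on the native HUDLayoutToolV2_LinkAcrossActionSetting class.
# Over a translucent backdrop, an icon and uppercase scrolling action name fill the left side.
# The right side holds a warning indicator, a color-coded add/remove impact count, and a bordered checkmark button for selecting the action.
# Its event graph has only unconnected event stubs, and key widgets are inherited from the parent.
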WidgetBlueprint /Game/Athena/HUD/Bacchus/HUDLayoutTool/LinkAcross/HUDLayoutToolV2_LinkAcrossActionSetting
graphs [EventGraph {nodes [
   Event PreConstruct()
   Event Construct()
   Event Tick()
  ]}]
variables []
widgets {
 info {
  inheritedWidgetCount=6
  parentClass=/Script/FortniteUI.HUDLayoutToolV2_LinkAcrossActionSetting
  rootWidgetClass=&SizeBox
  widgetCount=15
 }
 widgets [
  SizeBox_0 {widgetClassPath=&SizeBox, widget=SizeBox_0, properties {bOverride_HeightOverride=true, heightOverride=80, minDesiredHeight=60}}
  Overlay_0 {widgetClassPath=&Overlay, parent=SizeBox_0, widget=Overlay_0, slot {class=SizeBoxSlot}}
  Image_191 {
   widgetClassPath=&Image
   bIsVariable=true
   parent=Overlay_0
   widget=Image_191
   properties {
    visibility=SelfHitTestInvisible
    brush {
     bIsDynamicallyLoaded=false
     drawAs=Image
     imageType=NoImage
     mirroring=NoMirror
     resourceName=None
     resourceObject=null
     tiling=NoTile
     imageSize=<32 60>
     margin=<ltrb 0 0 0 0>
     outlineSettings {
      bUseBrushTransparency=false
      roundingType=HalfHeightRadius
      width=0
      color {colorUseRule=UseColor_Specified, specifiedColor=<rgba 0 0 0 0>}
      cornerRadii=<0 0 0 0>
     }
     tintColor {colorUseRule=UseColor_Specified, specifiedColor=<rgba 1 1 1 1>}
     uVRegion {bIsValid=false, max=<0 0>, min=<0 0>}
    }
    colorAndOpacity=<rgba 0 0 0 0.2>
   }
   slot {class=OverlaySlot, properties {horizontalAlignment=HAlign_Fill, verticalAlignment=VAlign_Fill}}
  }
  HorizontalBox_0 {
   widgetClassPath=&HorizontalBox
   parent=Overlay_0
   widget=HorizontalBox_0
   slot {class=OverlaySlot, properties {horizontalAlignment=HAlign_Fill, verticalAlignment=VAlign_Fill, padding=<ltrb 10 10 10 10>}}
  }
  HorizontalBox_27 {
   widgetClassPath=&HorizontalBox
   parent=HorizontalBox_0
   widget=HorizontalBox_27
   slot {class=HorizontalBoxSlot, properties {size {sizeRule=Fill, value=1}}}
  }
  SizeBox_4 {
   widgetClassPath=&SizeBox
   parent=HorizontalBox_27
   widget=SizeBox_4
   properties {
    bOverride_HeightOverride=true
    bOverride_WidthOverride=true
    heightOverride=60
    maxDesiredWidth=118.1081
    minDesiredWidth=60
    widthOverride=90
   }
   slot {class=HorizontalBoxSlot, properties {padding=<ltrb 0 0 14 0>}}
  }
  ScaleBox_0 {
   widgetClassPath=&ScaleBox
   parent=SizeBox_4
   widget=ScaleBox_0
   properties {stretch=ScaleToFitX}
   slot {class=SizeBoxSlot, properties {horizontalAlignment=HAlign_Center, verticalAlignment=VAlign_Center}}
  }
  Image_ActionIcon {
   widgetClassPath=&Image
   bInherited=true
   bIsVariable=true
   parent=ScaleBox_0
   widget=Image_ActionIcon
   properties {brush {
     bIsDynamicallyLoaded=false
     drawAs=Image
     imageType=NoImage
     mirroring=NoMirror
     resourceName=None
     resourceObject=/Game/Athena/HUD/Bacchus/HUDLayoutTool/HLT_Icons/T-HLT-Vehicle-Cart_Sprite
     tiling=NoTile
     imageSize=<68 65>
     margin=<ltrb 0 0 0 0>
     outlineSettings {
      bUseBrushTransparency=false
      roundingType=HalfHeightRadius
      width=0
      color {colorUseRule=UseColor_Specified, specifiedColor=<rgba 0 0 0 0>}
      cornerRadii=<0 0 0 0>
     }
     tintColor {colorUseRule=UseColor_Specified, specifiedColor=<rgba 1 1 1 1>}
     uVRegion {bIsValid=false, max=<0 0>, min=<0 0>}
    }}
   slot {class=ScaleBoxSlot, properties {horizontalAlignment=HAlign_Fill, verticalAlignment=VAlign_Fill}}
  }
  Text_ActionName {
   widgetClassPath=&CommonTextBlock
   bInherited=true
   parent=HorizontalBox_27
   widget=Text_ActionName
   properties {
    scrollStyle=&TextScrollStyle-Base_C
    text=Copy and link to all
    textTransformPolicy=ToUpper
    font {
     bForceMonospaced=false
     bMaterialIsStencil=false
     fontMaterial=null
     fontObject=&BurbankSmall
     letterSpacing=0
     monospacedWidth=1
     size=26
     skewAmount=0
     typefaceFontName=Black
     outlineSettings {
      bApplyOutlineToDropShadows=false
      bMiteredCorners=false
      bSeparateFillAlpha=false
      outlineMaterial=null
      outlineSize=0
      outlineColor=<rgba 0 0 0 1>
     }
    }
    shadowOffset=<0 0>
   }
   slot {class=HorizontalBoxSlot, properties {verticalAlignment=VAlign_Center, size {sizeRule=Fill, value=1}}}
  }
  HorizontalBox_25 {
   widgetClassPath=&HorizontalBox
   parent=HorizontalBox_0
   widget=HorizontalBox_25
   slot {class=HorizontalBoxSlot, properties {horizontalAlignment=HAlign_Right}}
  }
  LinkAcrossWarningIndicator_Warning {
   widgetClassPath=&/HUDLayoutToolV2_LinkAcrossWarningIndicator
   bInherited=true
   bIsVariable=true
   parent=HorizontalBox_25
   widget=LinkAcrossWarningIndicator_Warning
   properties {colorAndOpacity=<rgba 0.9301 1 0.0097 1>}
   slot {class=HorizontalBoxSlot, properties {padding=<ltrb 14 0 0 0>}}
  }
  NumericalIndicator_ActionImpact {
   widgetClassPath=&/HUDLayoutToolV2_LinkAcrossNumericalIndicator
   bInherited=true
   bIsVariable=true
   parent=HorizontalBox_25
   widget=NumericalIndicator_ActionImpact
   properties {
    actionSymbols [
     {actionSymbol=+, actionType=Adding, actionColor=<rgba 0.1529 0.6867 0.0685 1>}
     {actionSymbol=-, actionType=Removing, actionColor=<rgba 0.6939 0.0137 0.016 1>}
     {actionSymbol="", actionType=Nothing, actionColor=<rgba 1 1 1 1>}
    ]
    maximumInstancesColor=<rgba 0.6941 0.0118 0.0157 1>
   }
   slot {class=HorizontalBoxSlot, properties {verticalAlignment=VAlign_Center, padding=<ltrb 14 0 0 0>}}
  }
  Overlay_6 {
   widgetClassPath=&Overlay
   parent=HorizontalBox_25
   widget=Overlay_6
   slot {class=HorizontalBoxSlot, properties {padding=<ltrb 14 0 0 0>}}
  }
  Image_CheckboxBorder {
   widgetClassPath=&Image
   bInherited=true
   parent=Overlay_6
   widget=Image_CheckboxBorder
   properties {
    visibility=SelfHitTestInvisible
    brush {
     bIsDynamicallyLoaded=false
     drawAs=Image
     imageType=NoImage
     mirroring=NoMirror
     resourceName=None
     resourceObject=/Game/Athena/HUD/Bacchus/HUDLayoutTool/HLT_Icons/M_HUDLayoutToolV2_TiltSquare_Inst
     tiling=NoTile
     imageSize=<60 60>
     margin=<ltrb 0 0 0 0>
     outlineSettings {
      bUseBrushTransparency=false
      roundingType=HalfHeightRadius
      width=0
      color {colorUseRule=UseColor_Specified, specifiedColor=<rgba 0 0 0 0>}
      cornerRadii=<0 0 0 0>
     }
     tintColor {colorUseRule=UseColor_Specified, specifiedColor=<rgba 1 1 1 1>}
     uVRegion {bIsValid=false, max=<0 0>, min=<0 0>}
    }
    colorAndOpacity=<rgba 0 0.117 1 1>
   }
   slot {class=OverlaySlot, properties {horizontalAlignment=HAlign_Center, verticalAlignment=VAlign_Center}}
  }
  Button_SelectAction {
   widgetClassPath=/Game/Athena/HUD/Bacchus/HUDLayoutTool/HUDLayoutToolV2_Button
   bInherited=true
   parent=Overlay_6
   widget=Button_SelectAction
   properties {
    bOverride_HeightOverride=true
    bOverride_WidthOverride=true
    buttonHeightOverride=60
    buttonWidthOverride=60
    text=""
    backgroundBrush {
     bIsDynamicallyLoaded=false
     drawAs=NoDrawType
     imageType=NoImage
     mirroring=NoMirror
     resourceName=None
     resourceObject=null
     tiling=NoTile
     imageSize=<32 32>
     margin=<ltrb 0 0 0 0>
     outlineSettings {
      bUseBrushTransparency=false
      roundingType=HalfHeightRadius
      width=0
      color {colorUseRule=UseColor_Specified, specifiedColor=<rgba 0 0 0 0>}
      cornerRadii=<0 0 0 0>
     }
     tintColor {colorUseRule=UseColor_Specified, specifiedColor=<rgba 1 1 1 1>}
     uVRegion {bIsValid=false, max=<0 0>, min=<0 0>}
    }
    iconBrush {
     bIsDynamicallyLoaded=false
     drawAs=Image
     imageType=NoImage
     mirroring=NoMirror
     resourceName=None
     resourceObject=/Game/Athena/HUD/Bacchus/HUDLayoutTool/HLT_Icons/T-HLT-CheckMark-32_Sprite
     tiling=NoTile
     imageSize=<37 36>
     margin=<ltrb 0 0 0 0>
     outlineSettings {
      bUseBrushTransparency=false
      roundingType=HalfHeightRadius
      width=0
      color {colorUseRule=UseColor_Specified, specifiedColor=<rgba 0 0 0 0>}
      cornerRadii=<0 0 0 0>
     }
     tintColor {colorUseRule=UseColor_Specified, specifiedColor=<rgba 1 1 1 1>}
     uVRegion {bIsValid=false, max=<0 0>, min=<0 0>}
    }
    internalPadding=<ltrb 0 0 0 0>
   }
   slot {class=OverlaySlot, properties {horizontalAlignment=HAlign_Center, verticalAlignment=VAlign_Center}}
  }
 ]
}
