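# Title widget for the HUD Layout Tool V2 "link across" panel, showing which HUD widget is being copied and linked.
# Lays out a fixed-height preview box beside a width-capped column with the uppercase widget name above a grey "Copy and link to..." action line.
# The preview box and both texts are inherited bindings of the native HUDLayoutToolV2_LinkAcrossPanelTitle parent; the Blueprint only styles them and has no graph logic.
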WidgetBlueprint /FortniteCoreUI/Athena/HUD/Bacchus/HUDLayoutTool/LinkAcross/HUDLayoutToolV2_LinkAcrossPanelTitle
graphs [EventGraph {nodes [
   Event PreConstruct()
   Event Construct()
   Event Tick()
  ]}]
variables []
widgets {
 info {
  inheritedWidgetCount=3
  parentClass=/Script/FortniteUI.HUDLayoutToolV2_LinkAcrossPanelTitle
  rootWidgetClass=&VerticalBox
  widgetCount=7
 }
 widgets [
  VerticalBox_1 {widgetClassPath=&VerticalBox, widget=VerticalBox_1}
  HorizontalBox_25 {widgetClassPath=&HorizontalBox, parent=VerticalBox_1, widget=HorizontalBox_25, slot {class=VerticalBoxSlot}}
  SizeBox_Preview {
   widgetClassPath=&SizeBox
   bInherited=true
   parent=HorizontalBox_25
   widget=SizeBox_Preview
   properties {bOverride_HeightOverride=true, heightOverride=86, visibility=Visible}
   slot {class=HorizontalBoxSlot}
  }
  SizeBox_2 {
   widgetClassPath=&SizeBox
   parent=HorizontalBox_25
   widget=SizeBox_2
   properties {bOverride_WidthOverride=true, widthOverride=700}
   slot {class=HorizontalBoxSlot}
  }
  VerticalBox_0 {widgetClassPath=&VerticalBox, parent=SizeBox_2, widget=VerticalBox_0, slot {class=SizeBoxSlot}}
  Text_WidgetName {
   widgetClassPath=&CommonTextBlock
   bInherited=true
   parent=VerticalBox_0
   widget=Text_WidgetName
   properties {
    clipping=OnDemand
    scrollStyle=&TextScrollStyle-Base_C
    text=Widget Name
    textTransformPolicy=ToUpper
    font {
     bForceMonospaced=false
     bMaterialIsStencil=false
     fontMaterial=null
     fontObject=&BurbankSmall
     letterSpacing=0
     monospacedWidth=1
     size=26
     skewAmount=0
     typefaceFontName=Black
     outlineSettings {
      bApplyOutlineToDropShadows=false
      bMiteredCorners=false
      bSeparateFillAlpha=false
      outlineMaterial=null
      outlineSize=0
      outlineColor=<rgba 0 0 0 1>
     }
    }
    shadowOffset=<0 0>
   }
   slot {class=VerticalBoxSlot, properties {verticalAlignment=VAlign_Top, size {sizeRule=Fill, value=1}}}
  }
  Text_Action {
   widgetClassPath=&CommonTextBlock
   bInherited=true
   parent=VerticalBox_0
   widget=Text_Action
   properties {
    scrollStyle=&TextScrollStyle-Base_C
    text=Copy and link to...
    colorAndOpacity {colorUseRule=UseColor_Specified, specifiedColor=<rgba 0.4397 0.4397 0.4397 1>}
    font {
     bForceMonospaced=false
     bMaterialIsStencil=false
     fontMaterial=null
     fontObject=&BurbankSmall
     letterSpacing=0
     monospacedWidth=1
     size=22
     skewAmount=0
     typefaceFontName=Bold
     outlineSettings {
      bApplyOutlineToDropShadows=false
      bMiteredCorners=false
      bSeparateFillAlpha=false
      outlineMaterial=null
      outlineSize=0
      outlineColor=<rgba 0 0 0 1>
     }
    }
    shadowOffset=<0 0>
   }
   slot {class=VerticalBoxSlot, properties {verticalAlignment=VAlign_Bottom, size {sizeRule=Fill, value=1}}}
  }
 ]
}
